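# Editor utility asset action that bulk-converts the release version of selected FortVehicleCosmeticsItemDefinition assets.
# ConvertReleaseVersion edits from/to versions in an object dialog, backed by an SG_ConvertVehicleRelease SaveGame that persists them between runs.
# It rejects a source version newer than the target, with "Future" mapped to zero to pass that check.
# It filters the selection by class, re-versions definitions matching the source, saves them and prints the converted count.
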EditorUtilityBlueprint /VehicleCosmetics/EditorUtils/EAAU_VehicleItemDefActions
graphs [
 ConvertReleaseVersion {nodes [
   ConvertReleaseVersion() {is=title K2Node_FunctionEntry}
   > n1 = Sequence!(execute) -> (then_0, then_1, then_2:Exec, then_3:Exec)
   > n2 = Sequence()
   > n3 = LoadGameFromSlot!(execute, SlotName:String=Get l_savegameSlotName() -> (l_savegameSlotName:String), UserIndex:int=0) -> (then, ReturnValue:Save Game Object Reference)
   > n4 = Cast To SG_ConvertVehicleRelease!(execute, Object=@n3) -> (then, CastFailed, AsSG Convert Vehicle Release:SG Convert Vehicle Release Object Reference) {is=title K2Node_DynamicCast}
   > Set l_savegame(l_savegame:SG Convert Vehicle Release Object Reference=@n4) -> (Output_Get:SG Convert Vehicle Release Object Reference)
   n5 = CreateSaveGameObject!(
    execute=@n4.CastFailed
    SaveGameClass:Save Game Class Reference=&/SG_ConvertVehicleRelease.SG_ConvertVehicleRelease_C
   ) -> (then, ReturnValue:SG Convert Vehicle Release Object Reference)
   > Set l_savegame(l_savegame=@n5)
   n6 = Get l_savegame() -> (l_savegame:SG Convert Vehicle Release Object Reference)
   Set l_toFloat(
    execute=@n2.then_1
    l_toFloat:double=String To Float(InString=Get ReleaseToConvertTo!(self:SG Convert Vehicle Release Object Reference=@n6) -> (ReleaseToConvertTo:String))
   ) -> (Output_Get:double)
   n7 = Get ReleaseToConvertFrom!(self:SG Convert Vehicle Release Object Reference=Get l_savegame()) -> (ReleaseToConvertFrom:String)
   > n8 = Branch(Condition="Equal Exactly (String)"(A=@n7, B=Future))
   > Set l_fromFloat(l_fromFloat:double=0) -> (Output_Get:double)
   Set l_fromFloat(execute=@n8.else, l_fromFloat=String To Float(InString=@n7))
   n9 = Show Object Dialog!(
    execute=@n2.then_2
    Title:"Text (by ref)"=MakeLiteralText(Value="INVTEXT(\"Conversion Settings.\")")
    InOutObject:Object Reference=Get l_savegame()
    Options:"Editor Dialog Library Object Details View Options Structure (by ref)"=Make Editor Dialog Library Object Details View Options!(
     bShowObjectName:bool=false
     bAllowSearch:bool=false
     bAllowResizing:bool=true
     MinWidth:int=0
     MinHeight:int=0
     ValueColumnWidthRatio:float=0.6
    ) -> (EditorDialogLibraryObjectDetailsViewOptions:Editor Dialog Library Object Details View Options Structure) {is=title K2Node_MakeStruct}
   ) -> (then, ReturnValue:bool)
   > n10 = Branch(Condition=@n9)
   > Branch(Condition="float > float"(A=Get l_fromFloat() -> (l_fromFloat:double), B=Get l_toFloat() -> (l_toFloat:double)))
   n11 = Get l_savegame()
   n12 = Get ReleaseToConvertFrom(self=@n11)
   n13 = Get ReleaseToConvertTo(self=@n11)
   > PrintString(InString="To String (Text)"(InText=Format Text!(
      Format="LOCTEXT(\"VehicleIDConversion: Tried to run conversion but the Release Version ({from})to convert from is newer than the release version to convert to ({to})\")"
      from:String=@n12
      to:String=@n13
     ) -> (Result)))
   > Return Node()
   Return Node(execute=@n10.else)
   SaveGameToSlot!(execute=@n2.then_3, SaveGameObject:Save Game Object Reference=Get l_savegame(), SlotName:String=Get l_savegameSlotName(), UserIndex:int=0) -> (then, ReturnValue:bool)
   n14 = GetSelectedAssetData(execute=@n1.then_1)
   > n15 = Set l_VehicleItemDefs(l_VehicleItemDefs:Array of Asset Data Structures=@n14) -> (Output_Get:Array of Asset Data Structures)
   n16 = MakeTopLevelAssetPath!(PackageName:String=/Script/VehicleCosmeticsRuntime, AssetName:String=FortVehicleCosmeticsItemDefinition) -> (ReturnValue:Top Level Asset Path Structure)
   > RunAssetsThroughFilter!(
    execute
    self:Asset Registry Interface=GetAssetRegistry()
    AssetDataList:Array of Asset Data Structures=@n15
    Filter:"ARFilter Structure (by ref)"=MakeARFilter!(
     PackageNames:Array of Names
     PackagePaths:Array of Names=Get l_paths() -> (l_paths:Array of Names)
     SoftObjectPaths:Array of Soft Object Path Structures
     ClassPaths:Array of Top Level Asset Path Structures=Make Array("[0]":Top Level Asset Path Structure=@n16) -> (Array:Array of Top Level Asset Path Structures)
     RecursiveClassPathsExclusionSet:Set of Top Level Asset Path Structures
     ClassNames:Array of Names
     RecursiveClassesExclusionSet:Set of Names
     bRecursivePaths:bool=true
     bRecursiveClasses:bool=true
     bIncludeOnlyOnDiskAssets:bool=false
    ) -> (ReturnValue:ARFilter Structure)
   ) -> (then)
   n17 = Length(TargetArray:Array of Asset Data Structures=@n15)
   > n18 = Branch(Condition="integer > integer"(A=@n17))
   > PrintString(InString="To String (Text)"(InText=Format Text!(Format="LOCTEXT(\"VehicleIDConversion: No assets selected of type FortVehicleCosmeticsItemDefinition.\")") -> (Result)))
   n19 = Get l_savegame()
   n20 = Get ReleaseToConvertFrom(self=@n19)
   n21 = Get ReleaseToConvertTo(self=@n19)
   PrintString(
    execute=@n18.else
    InString="To String (Text)"(InText=Format Text!(
      Format="LOCTEXT(\"VehicleIDConversion: Converting {num} item definitions from {from} to {to}.\")"
      num:int=@n17
      from:String=@n20
      to:String=@n21
     ) -> (Result))
   )
   n22 = For Each Loop(Exec=@n1.then_2, Array:Array of Asset Data Structures=Get l_VehicleItemDefs() -> (l_VehicleItemDefs:Array of Asset Data Structures)) -> (Array Element:Asset Data Structure)
   > n23 = Cast To FortVehicleCosmeticsItemDefinition!(execute, Object=GetAsset(InAssetData=@n22)) -> (then, CastFailed, AsFort Vehicle Cosmetics Item Definition:Fort Vehicle Cosmetics Item Definition Object Reference) {is=title K2Node_DynamicCast}
   > n24 = EditorOnly_GetItemDefinitionFortReleaseVersion!(execute, ItemDefinition:Item Definition Base Object Reference=@n23) -> (then, ReturnValue:Fort Release Version Structure)
   > n25 = EditorOnly_GetVersionDisplayTextForUIPurposes!(execute, Version:"Fort Release Version Structure (by ref)"=@n24) -> (then, ReturnValue:Text)
   n26 = Get l_savegame()
   > Branch(Condition="Equal Exactly (String)"(A="To String (Text)"(InText=@n25), B=Get ReleaseToConvertFrom(self=@n26)))
   > n27 = EditorOnly_BuildFortReleaseVersionFromString!(execute, VersionString:String=Get ReleaseToConvertTo(self=Get l_savegame())) -> (then, OutVersion:Fort Release Version Structure, ReturnValue:bool)
   > n28 = SetFortReleaseVersion!(execute, ItemDefinition:Fort Item Definition Object Reference=@n23, FortReleaseVersion:"Fort Release Version Structure (by ref)"=@n27) -> (then, ReturnValue:bool)
   > Branch(Condition=@n28)
   > Add(
    TargetArray:Array of Object References=Get l_assetsToSave() -> (l_assetsToSave:Array of Object References)
    NewItem:"Object Reference (by ref)"=@n23
   ) {is=title K2Node_CallArrayFunction}
   n29 = Get l_assetsToSave()
   n30 = Length(TargetArray:Array of Object References=@n29)
   n31 = Branch(execute=@n1.then_3, Condition="integer > integer"(A=@n30))
   > SaveLoadedAssets(AssetsToSave=@n29, bOnlyIfIsDirty=true)
   n32 = Length(TargetArray:Array of Asset Data Structures=Get l_VehicleItemDefs())
   > PrintString(InString="To String (Text)"(InText=Format Text!(
      Format="LOCTEXT(\"VehicleIDConversion: Successfully converted {num_conv} out of {num_sel} selected item defs.\")"
      num_conv:int=@n30
      num_sel:int=@n32
     ) -> (Result)))
   PrintString(execute=@n31.else, InString=VehicleIDConversion: Found no valid item definitions to convert)
  ]}
 EventGraph {links [], nodes []}
]
variables []
